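# Looping music cue for the Shrimp Shop in the NeonCity POI.
# A single looping wave player feeds the MusicParty POI audio bus and the Fortnite music audio-analysis submix.
# Base submix output is disabled, so the cue's audio travels through those sends instead of the normal output path.
# It references the NeonCity diegetic music concurrency set, which by its name groups it with the POI's other in-world music.
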SoundCue /Game/Sounds/Asteria/POI/NeonCity/mx_NeonCity/MX_SoundCue_POI/Mx_NeonCity_ShrimpShop_Loop_Cue
root=WavePlayer_1
nodes [{
  id=WavePlayer_1
  properties {bLooping=true, soundWaveAssetPtr=&/MX_S24_ShrimpShopLoop}
 }]
properties {
 bEnableBaseSubmix=false
 maxDistance=2097152
 volumeMultiplier=1
 attenuationOverrides {
  bSpatialize=false
  falloffDistance=500
  nonFocusVolumeAttenuation=0
  nonSpatializedRadiusStart=500
  stereoSpread=1500
  attenuationShapeExtents {x=600}
 }
 busSends [{
   audioBus=/Game/Sounds/Fort_Audio_Blueprints/MusicParty/MusicParty_POI_AudioBus
   maxSendDistance=1
   minSendDistance=1
   minSendLevel=1
   sendLevel=1
   soundSourceBus=null
  }]
 concurrencyOverrides {bVolumeScaleCanRelease=true, maxCount=1, resolutionRule=StopQuietest}
 concurrencySet [&/NeonCity_DiegeticMusicConcurrency]
 soundSubmixSends [{minSendLevel=0, soundSubmix=&Fortnite_AudioAnalysis_Music}]
}
