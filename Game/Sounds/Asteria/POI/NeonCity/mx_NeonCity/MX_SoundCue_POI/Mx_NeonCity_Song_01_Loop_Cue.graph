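# Looping sound cue that plays Neon City POI music song 01, a 145 BPM full loop, as gameplay music.
# A single looping wave player feeds the output at reduced volume.
# Spatial falloff comes from the shared Neon City ramen sign attenuation asset, and voice limiting from the Neon City diegetic music concurrency set.
# Submix sends route it to the default reverb and to a music audio analysis submix.
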SoundCue /Game/Sounds/Asteria/POI/NeonCity/mx_NeonCity/MX_SoundCue_POI/Mx_NeonCity_Song_01_Loop_Cue
root=WavePlayer_2
nodes [{
  id=WavePlayer_2
  properties {bLooping=true, soundWaveAssetPtr=&/MX_NC_DD01_Full_Loop_145BPM}
 }]
properties {
 attenuationSettings=&/MX_NeonCity_RamenSign_Att
 maxDistance=4500
 soundClassObject=/Game/Sounds/FortSoundClassesAndModes/Fort_Music_Gameplay
 volumeMultiplier=0.4
 attenuationOverrides {
  bEnableOcclusion=true
  falloffDistance=925
  nonFocusVolumeAttenuation=0
  nonSpatializedRadiusStart=900
  occlusionLowPassFilterFrequency=2000
  occlusionTraceChannel=ECC_GameTraceChannel6
  occlusionVolumeAttenuation=0.7
  stereoSpread=1500
  attenuationShapeExtents {x=900}
 }
 concurrencyOverrides {bVolumeScaleCanRelease=true, maxCount=1}
 concurrencySet [&/NeonCity_DiegeticMusicConcurrency]
 soundSubmixSends [
  {minSendLevel=0.5, soundSubmix=&Fortnite_Reverb_Default}
  {minSendLevel=0, soundSubmix=&Fortnite_AudioAnalysis_Music}
 ]
}
